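# Sound cue for the looping servo motor mechanism of a FortilaLight spot light, classed as a gadget sound.
# A looping wave player feeds a modulator at default settings, which the root quality node plays for its first two slots, leaving the other two empty.
# Spot light attenuation and low volume keep it localized, while a concurrency override caps simultaneous voices at two with a short steal fade.
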SoundCue /Game/Sounds/Artemis/Environment_Actors/FortilaLight/Spot_Light_Mechanism_Loop_Cue
root=QualityLevel_0
nodes [
 {id=Modulator_0, children [WavePlayer_1]}
 {id=QualityLevel_0, children [Modulator_0,Modulator_0,null,null]}
 {
  id=WavePlayer_1
  properties {bLooping=true, soundWaveAssetPtr=&/Spot_Light_Servo_Motor_Mechanisim_Loop}
 }
]
properties {
 attenuationSettings=&/Spot_Light_Att
 bEnableBusSends=false
 bEnableSubmixSends=false
 bOverrideConcurrency=true
 maxDistance=665
 soundClassObject=&Fort_Gadgets
 volumeMultiplier=0.25
 concurrencyOverrides {maxCount=2, voiceStealReleaseTime=0.2}
}
