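# Sound cue for the startup servo-motor sound of a spotlight mechanism, classed as a gadget sound with its own spotlight attenuation.
# A Random node picks between two equally weighted wave variations, and a Modulator varies the result with a widened downward pitch range.
# The root Quality Level node routes only its first two slots to that chain, leaving the other two silent.
# A concurrency override limits simultaneous voices, and bus and submix sends are disabled.
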SoundCue /Game/Sounds/Artemis/Environment_Actors/FortilaLight/Spot_Light_Mechanism_Start_Cue
root=QualityLevel_0
nodes [
 {id=Modulator_0, children [Random_1], properties {pitchMin=0.9}}
 {id=QualityLevel_0, children [Modulator_0,Modulator_0,null,null]}
 {id=Random_1, children [WavePlayer_3,WavePlayer_4], properties {weights [1,1]}}
 {
  id=WavePlayer_3
  properties {soundWaveAssetPtr=&/Spot_Light_Servo_Motor_Mechanisim_Start_01}
 }
 {
  id=WavePlayer_4
  properties {soundWaveAssetPtr=&/Spot_Light_Servo_Motor_Mechanisim_Start_02}
 }
]
properties {
 attenuationSettings=&/Spot_Light_Att
 bEnableBusSends=false
 bEnableSubmixSends=false
 bOverrideConcurrency=true
 maxDistance=665
 soundClassObject=&Fort_Gadgets
 volumeMultiplier=0.3
 concurrencyOverrides {maxCount=2, voiceStealReleaseTime=0.2}
}
