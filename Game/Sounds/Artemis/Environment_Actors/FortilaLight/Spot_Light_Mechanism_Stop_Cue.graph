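# Sound cue for the stop sound of a spot light servo motor mechanism, classed under Fort_Gadgets.
# A random node picks one of two equally weighted stop waves and feeds a modulator.
# A quality-level root routes that modulator to its first two slots and leaves the other two empty.
# It borrows the RefuelingCart attenuation asset, plays at reduced volume, overrides concurrency to limit simultaneous voices, and disables bus and submix sends.
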SoundCue /Game/Sounds/Artemis/Environment_Actors/FortilaLight/Spot_Light_Mechanism_Stop_Cue
root=QualityLevel_0
nodes [
 {id=Modulator_0, children [Random_0]}
 {id=QualityLevel_0, children [Modulator_0,Modulator_0,null,null]}
 {id=Random_0, children [WavePlayer_1,WavePlayer_2], properties {weights [1,1]}}
 {
  id=WavePlayer_1
  properties {soundWaveAssetPtr=&/Spot_Light_Servo_Motor_Mechanisim_Stop_01}
 }
 {
  id=WavePlayer_2
  properties {soundWaveAssetPtr=&/Spot_Light_Servo_Motor_Mechanisim_Stop_02}
 }
]
properties {
 attenuationSettings=/Game/Sounds/Artemis/Environment_Actors/RefuelingCart/Spot_RefuelingCart_Att
 bEnableBusSends=false
 bEnableSubmixSends=false
 bOverrideConcurrency=true
 maxDistance=750
 soundClassObject=&Fort_Gadgets
 volumeMultiplier=0.3
 concurrencyOverrides {maxCount=2, voiceStealReleaseTime=0.2}
}
